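# Additive, unlit, two-sided material for Niagara mesh particles, drawing the audio ring in the BandPeakState effect.
# SplineThicken sets strand width, and blurred noise panned over UV1, time-offset per particle random, adds displacement along vertex normals.
# A "WPO Mult" dynamic parameter scales that displacement, letting the particle system drive the distortion.
# Particle color feeds an emissive adjustment function, and particle alpha multiplies a GeneratedBand mask on UV1.
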Material /PilgrimQuickplay/Effects/BandPeakState/Materials/M_BandPeak_AudioRing
graph {
 properties {
  bUsedWithNiagaraMeshParticles=true
  blendMode=BLEND_Additive
  pixelDepthOffsetMode=PDOM_Legacy
  shadingModel=MSM_Unlit
  twoSided=true
  usedShadingModels=Unlit
 }
 nodes [
  n1 = ScalarParameter() {defaultValue=50}
  n2 = VertexNormalWS()
  n3 = TextureCoordinate() {coordinateIndex=1}
  n4 = ScalarParameter() {defaultValue=1, parameterName=Param_2}
  n5 = Multiply(A=Time(), B=ScalarParameter() {defaultValue=0.2, parameterName=Param_3})
  n6 = ParticleRandom()
  n7 = TextureSample(UVs=Panner(Coordinate=Multiply(A=@n3, B=@n4), Time=Add(A=@n5, B=@n6)) {speedX=1}) {samplerType=SAMPLERTYPE_Grayscale, texture=&LowResBlurredNoise}
  n8 = ScalarParameter() {defaultValue=200, parameterName=Param_1}
  n9 = DynamicParameter() {paramNames [WPO Mult, Param2, Param3, Param4]}
  Add(
   A=MaterialFunctionCall(WidthBase=@n1, WidthTip=@n1) {materialFunction=&SplineThicken}.WorldPositionOffset
   B=Multiply(A=Multiply(A=@n2, B=@n7.R), B=Multiply(A=@n8, B=@n9.WPO Mult)) {constB=10}
  )
  n10 = ParticleColor()
  MaterialFunctionCall(Emissive=@n10.RGB) {materialFunction=&MF_FX_EmissiveAdjustments}
  Multiply(
   A=@n10.A
   B=MaterialFunctionCall!(Width, Sharpness, Offset, Direction Switch=StaticBool() {value=true}, Compare, Input Coordinates=TextureCoordinate() {coordinateIndex=1}) {materialFunction=/Engine/Functions/Engine_MaterialFunctions03/Procedurals/GeneratedBand}.Result
  )
 ]
}
